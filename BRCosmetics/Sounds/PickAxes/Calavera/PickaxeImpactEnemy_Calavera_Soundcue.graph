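# Sound cue for the Calavera pickaxe's impact on enemies, using shared pickaxe impact attenuation, property sheet and the impacts sound class.
# A root random node, with replacement, favors a set of six plain impact waves over six "Lyr" variants.
# Each branch applies a short fixed delay, then picks one of its waves at equal weight.
# Despite the "Lyr" naming, the two sets are alternatives and never play together.
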SoundCue /BRCosmetics/Sounds/PickAxes/Calavera/PickaxeImpactEnemy_Calavera_Soundcue
root=Random_2
nodes [
 {id=Delay_0, children [Random_0], properties {delayMax=0.1, delayMin=0.1}}
 {id=Delay_1, children [Random_1], properties {delayMax=0.1, delayMin=0.1}}
 {id=Random_0, children [WavePlayer_0,WavePlayer_1,WavePlayer_2,WavePlayer_3,WavePlayer_4,WavePlayer_5], properties {weights [1,1,1,1,1,1]}}
 {
  id=Random_1
  children [WavePlayer_6, WavePlayer_7, WavePlayer_8, WavePlayer_9, WavePlayer_10, WavePlayer_11]
  properties {weights [1,1,1,1,1,1]}
 }
 {id=Random_2, children [Delay_0,Delay_1], properties {bRandomizeWithoutReplacement=false, weights [1,0.3]}}
 {id=WavePlayer_0, properties {soundWaveAssetPtr=&/Pickaxe_Calavera_Impact_06}}
 {id=WavePlayer_1, properties {soundWaveAssetPtr=&/Pickaxe_Calavera_Impact_01}}
 {id=WavePlayer_10, properties {soundWaveAssetPtr=&/Pickaxe_Calavera_Impact_04_Lyr}}
 {id=WavePlayer_11, properties {soundWaveAssetPtr=&/Pickaxe_Calavera_Impact_05_Lyr}}
 {id=WavePlayer_2, properties {soundWaveAssetPtr=&/Pickaxe_Calavera_Impact_02}}
 {id=WavePlayer_3, properties {soundWaveAssetPtr=&/Pickaxe_Calavera_Impact_03}}
 {id=WavePlayer_4, properties {soundWaveAssetPtr=&/Pickaxe_Calavera_Impact_04}}
 {id=WavePlayer_5, properties {soundWaveAssetPtr=&/Pickaxe_Calavera_Impact_05}}
 {id=WavePlayer_6, properties {soundWaveAssetPtr=&/Pickaxe_Calavera_Impact_06_Lyr}}
 {id=WavePlayer_7, properties {soundWaveAssetPtr=&/Pickaxe_Calavera_Impact_01_Lyr}}
 {id=WavePlayer_8, properties {soundWaveAssetPtr=&/Pickaxe_Calavera_Impact_02_Lyr}}
 {id=WavePlayer_9, properties {soundWaveAssetPtr=&/Pickaxe_Calavera_Impact_03_Lyr}}
]
properties {
 attenuationSettings=/Game/Sounds/Attenuation/Fort_Pickaxe_Impact_Att
 audioPropertiesSheet=/BRCosmetics/Sounds/Settings/PropertySheets/Pickaxe/APS_Cosmetic_Pickaxe_Impact
 maxDistance=4300
 soundClassObject=&Fort_Impacts
 concurrencySet [null]
}
